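# Niagara skydiving contrail cosmetic FoundKelp, drawing user-colored sprite trails and streaks from skeletal mesh sockets.
# Contrail setup dive value and direction, plus glide lean, drive spawn rate, size, motion and brightness.
# High lean spawns a looping logo sprite that rises, brakes past a height threshold, fades and fires a location event.
# A rebuilt impact spark emitter adds velocity-aligned sparks; the original and a hand glow emitter are disabled.
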NiagaraSystem /BRCosmetics/Athena/Effects/Particles/Skydiving/NS_Contrail_FoundKelp/NS_Contrail_FoundKelp
dependencies {
 usedDataInterfaces [
  &NiagaraDataInterfaceCurve
  &NiagaraDataInterfaceArrayFloat
  &NiagaraDataInterfaceArrayInt32
  &NiagaraDataInterfaceVector2DCurve
  &NiagaraDataInterfaceVectorCurve
  &NiagaraDataInterfaceColorCurve
  &NiagaraDataInterfaceCamera
  &NiagaraDataInterfaceSkeletalMesh
 ]
 usedDynamicInputs []
 usedModules [
  &self:SystemSpawnScript.SystemScriptSource.SystemScriptGraph.NiagaraNodeAssignment_0.SetVariables_#1
  &SystemState
  /Game/Effects/Niagara/Modules/Skydiving/FN_DivingAttributes/v5/FN_DivingAttributes
  /Game/Effects/Niagara/Modules/Contrails/FN_Contrail_Setup
  &FN_Integrator
  &self:SystemSpawnScript.SystemScriptSource.SystemScriptGraph.NiagaraNodeAssignment_1.SetVariables_#2
 ]
}
emitters {
 DigitalSquares001 {
  emitterUpdateScript [
   &EmitterState()
   &SpawnRate(SpawnRate=&Lerp_Float(A=30, B=50, Alpha=$System.FN_Contrail_Setup.Dive Value), Use Spawn Probability=false) {module=SpawnRate002}
  ]
  particleSpawnScript [
   &PartitionParticles()
   &InitializeParticle(
    Color=&MultiplyLinearColor(
     Linear Color=&RandomRangeLinearColor(Minimum=$User.FXColorB, Maximum=$User.FXColorA)
     Scale Factor=<5 5 5 1>
    )
    Lifetime=&Lerp_Float(
     A=&RandomRangeFloat(Minimum=0.3, Maximum=0.5)
     B=&RandomRangeFloat(Minimum=0.15, Maximum=0.3)
     Alpha=$System.FN_Contrail_Setup.Dive Value
    )
    Sprite Size Mode=::NewEnumerator1
    Uniform Sprite Size Min=2
    UsePositionOffset=false
    Write Material Random=true
   )
   &SkeletalMeshLocation(
    Bone / Socket Index=$Particles.PartitionParticles.Partition
    MaskSpawnGroup=false
    Mesh Sampling Type=::NewEnumerator5
    Orientation Sampling=::NewEnumerator1
    Sampled Position Offset=&Multiply_Vector(
     A=&Lerp_Vector(A=<0 0 -1>, B=<1 1 1>, Alpha=&MakeFloatFromInt(Int=$Particles.PartitionParticles.Partition))
     B=&Lerp_Vector(A=<1 1 1>, B=<1 1 1>, Alpha=$User.ParaGlideLeanAlpha)
    )
    Skeletal Mesh=$System.Skeletal Mesh
    Socket Sampling Mode=::NewEnumerator5
   )
   &ShapeLocation(
    MaskSpawnGroup=false
    Override Seed=false
    Sphere Radius=&Lerp_Float(A=15, B=20, Alpha=$System.FN_Contrail_Setup.Dive Value)
    UseOwnerScale=true
   )
   &AddVelocity(Velocity=&Lerp_Vector(A=<0 0 1>, B=<0 0 1>, Alpha=$System.FN_Contrail_Setup.Dive Value), Velocity Speed Scale=500)
   &ApplyInitialForces()
  ]
  particleUpdateScript [
   &ParticleState()
   &AccelerationForce(
    Acceleration=&Lerp_Vector(
     A=&RandomRangeVector(Minimum=<100 100 100>, Maximum=<-100 -100 300>)
     B=&Multiply_VectorByFloat(Vector=$System.FN_Contrail_Setup.Direction Velocity, Float=10)
     Alpha=$User.ParaGlideLeanAlpha
    )
    Coordinate Space=::Local
   ) {enabled=false, module=AccelerationForce001}
   &AccelerationForce(
    Acceleration=&Multiply_Vector(
     A=$System.FN_Contrail_Setup.Direction Velocity
     B=&Lerp_Vector(A=<5 5 5>, B=<10 10 10>, Alpha=$User.ParaGlideLeanAlpha)
    )
    Coordinate Space=::Local
   )
   &UpdateMeshOrientation(
    Facing Direction=&Lerp_Vector(A=$System.FN_Contrail_Setup.Direction Vector, B=$Particles.Velocity, Alpha=0)
    Facing Interpolation=false
    Orientation Method=::NewEnumerator1
    Override Quat=false
    Stabilization Mode:ENiagara_OrientVectorCount=::NewEnumerator2
    TransformFacing=false
    TransformSide=false
   )
   &ScaleVelocity(Velocity Scale=&Add_Vector(
     A=&Multiply_Vector(A=$System.FN_Contrail_Setup.Direction Vector, B=<1 1 -0.1>)
     B=&Lerp_Vector(A=<0 0 0>, B=&RandomRangeVector(Minimum=<0 0 0>, Maximum=<0 0 0>), Alpha=$User.ParaGlideLeanAlpha)
    ))
   &SolveForcesAndVelocity(Clamp Velocity=false, Limit Acceleration=false)
   &AlignSpriteToMeshOrientation(
    Mesh Orientation Relative Sprite Alignment Vector=<1 0 0>
    Mesh Orientation Relative Sprite Facing Vector:Vector3f=<0 1 0>
    SetAlignment=true
    SetFacing=false
   )
   &ScaleSpriteSizeBySpeed(
    Max Scale Factor=&Multiply_Vector2D_ByFloat(
     Float=1
     Vector2D=&Multiply_Vector2D_ByFloat(Float=0.5, Vector2D=&RandomRangeVector2D(Minimum=<0.1 0.2>, Maximum=<0.2 10>))
    )
    Min Scale Factor=<0.3 1.5>
   )
   &ScaleSpriteSize(
    Non-Uniform Curve Scale=&Multiply_Vector2D_ByFloat(
     Float=&FloatFromCurve(FloatCurve=DataInterface() {propertyValues {Curve {
         defaultValue=0
         keys [
          <at 0 0 RCIM_Cubic>
          {arriveTangent=2, time=0.1, value=1}
          {interpMode=RCIM_Cubic, tangentMode=RCTM_Break, time=0.5, value=1}
          {interpMode=RCIM_Cubic, tangentMode=RCTM_User, time=1, value=1}
         ]
        }}})
     Vector2D=<1 1>
    )
    Non-Uniform Curve Sprite Scale=DataInterface() {propertyValues {
      XCurve {keys [{interpMode=RCIM_Cubic, tangentMode=RCTM_User, time=0, value=1}, {tangentMode=RCTM_User, time=1, value=1}]}
      YCurve {keys [
        {arriveTangent=4.4047, interpMode=RCIM_Cubic, leaveTangent=4.4047, tangentMode=RCTM_User, time=-0.005, value=0.1}
        {arriveTangent=0.1609, leaveTangent=1.7823, tangentMode=RCTM_User, time=0.3, value=1}
        {interpMode=RCIM_Cubic, tangentMode=RCTM_User, time=0.5, value=1}
        {tangentMode=RCTM_User, time=1, value=0}
       ]}
     }}
    Scale Sprite Size Mode=::NewEnumerator3
   )
   &CameraOffset(Camera Offset Amount=&RandomRangeFloat(Minimum=5, Maximum=15))
   &Color(
    Color=&Multiply_LinearColorByFloat(
     Linear Color=&LerpLinearColors(StartColor=$User.FXColorB, EndColor=$User.FXColorA, LerpFactor=$User.ParaGlideLeanAlpha)
     Float=&FloatFromCurve(
      FloatCurve=DataInterface() {propertyValues {Curve {keys [{interpMode=RCIM_Cubic, tangentMode=RCTM_User, time=0, value=2}, <at 1 10>]}}}
      CurveIndex=$User.ParaGlideLeanAlpha
      Scale Curve=&RandomRangeFloat(Minimum=0.1, Maximum=1)
     )
    )
    Scale Alpha=10
    Scale Color=&Multiply_VectorByFloat(
     Vector=<1 1 1>
     Float=&FloatFromCurve(FloatCurve=DataInterface() {propertyValues {Curve {keys [{arriveTangent=-17.669, interpMode=RCIM_Cubic, leaveTangent=-17.669, tangentMode=RCTM_User, time=0, value=100}, <at 0.2 1>, <at 1 1>]}}})
    )
   )
   &CameraOffset(Camera Offset Amount=&RandomRangeFloat(Minimum=15, Maximum=-3)) {module=CameraOffset001}
  ]
  rendererClasses [&NiagaraSpriteRendererProperties]
  renderers [{rendererClass=&NiagaraSpriteRendererProperties}]
 }
 FoundKelp_Logo_Mesh {
  emitterUpdateScript [
   &EmitterState(Delay First Loop Only=true, Life Cycle Mode=::NewEnumerator1, Loop Duration=2, UseLoopDelay=true)
   &SpawnBurst_Instantaneous(
    Spawn Count=&MakeCustomIntFromBool(
     Bool=&SetBoolByFloatComparison(A=$User.ParaGlideLeanAlpha, B=0.6, Comparison Type=::NewEnumerator2)
     True INT=1
     False INT=0
    )
    Use Spawn Probability=false
    UseLoopCountLimit=false
   )
  ]
  particleSpawnScript [
   &InitializeParticle(Sprite Size Mode=::NewEnumerator3, Uniform Sprite Size=215, UsePositionOffset=false, Write Material Random=true)
   &SystemLocation(
    MaskSpawnGroup=false
    Offset=&FN_FrontEndVector3Modifier(Gameplay=<0 25 -250>, Front End=<0 75 -250>)
   )
   &AddVelocity(Constrain To Radius=215.0099, Module.Use Falloff Distance=false, Override Seed=false, Velocity Mode=::NewEnumerator1, Velocity Speed=1) {module=AddVelocity001}
   &ApplyInitialForces()
   &CameraOffset(Camera Offset Amount=15)
  ]
  particleUpdateScript [
   &ParticleState()
   &ScaleSpriteSize(Uniform Curve Sprite Scale=DataInterface() {propertyValues {Curve {keys [
        {arriveTangent=1.8165, interpMode=RCIM_Cubic, leaveTangent=1.8165, tangentMode=RCTM_User, time=0, value=0.5}
        {tangentMode=RCTM_User, time=0.3, value=1}
       ]}}})
   &UpdateMeshOrientation(
    Facing Direction=&RotateVectorByQuaternion(
     Vector To Rotate=&Multiply_Vector(
      A=&ReturnMeshOrientationAxis(Mesh Orientation Axis=::YAxis, Mesh Orientation Quat=$Particles.MeshOrientation)
      B=<0 1 0>
     )
     Quaternion=<0.7 0 0 1>
    )
    Facing Interpolation=false
    Orientation Method=::NewEnumerator1
    Override Quat=false
    Stabilization Mode:ENiagara_OrientVectorCount=::NewEnumerator2
    TransformFacing=false
    TransformSide=false
   )
   /SetVariables(Particles.Collided:NiagaraBool=&SetBoolByFloatComparison(
     A=&MakeFloatFromVector(Vector=Linked() {linkedVariable {name=Particles.Position, type=NiagaraPosition}}, Channel=::NewEnumerator2)
     B=&Add_Float(
      A=&MakeFloatFromVector(Vector=Linked() {linkedVariable {name=Engine.Emitter.SimulationPosition, type=NiagaraPosition}}, Channel=::NewEnumerator2)
      B=&FN_FrontEndFloatModifier(
       Gameplay=&Lerp_Float(A=50, B=50, Alpha=$System.Dive)
       Front End=&Lerp_Float(A=50, B=50, Alpha=$System.Dive)
      )
     )
     Comparison Type=::NewEnumerator2
    )) {isSetParameters=true}
   &ScaleVelocity(Coordinate Space:ENiagaraCoordinateSpace=::Simulation) {enabled=false, module=ScaleVelocity005}
   &LinearForce(Force=&FN_FrontEndVector3Modifier(Gameplay=<0 -100 2000>, Front End=<0 -500 2000>))
   &Drag(
    Drag=&MakeCustomFloatFromBool(
     Bool=$Particles.Collided
     True Float=&FloatFromCurve(
      FloatCurve=DataInterface() {propertyValues {Curve {keys [
          {arriveTangent=101.6906, interpMode=RCIM_Cubic, tangentMode=RCTM_User, time=0, value=0}
          {interpMode=RCIM_Cubic, leaveTangent=2.7695, tangentMode=RCTM_Break, time=0.25, value=30}
          {arriveTangent=-10.5527, tangentMode=RCTM_User, time=1, value=3}
         ]}}}
      Scale Curve=&FN_FrontEndFloatModifier(Gameplay=1, Front End=1)
     )
     False Float=0
    )
    Use Linear Drag=true
    Use Rotational Drag=true
   ) {module=Drag001}
   &DoOnce(
    Reset:NiagaraBool=&SetBoolByFloatComparison(
     A=&MakeFloatFromVector(Vector=Linked() {linkedVariable {name=Particles.Position, type=NiagaraPosition}}, Channel=::NewEnumerator2)
     B=&Add_Float(
      A=&MakeFloatFromVector(Vector=Linked() {linkedVariable {name=Engine.Emitter.SimulationPosition, type=NiagaraPosition}}, Channel=::NewEnumerator2)
      B=&FN_FrontEndFloatModifier(
       Gameplay=&Lerp_Float(A=-55, B=-50, Alpha=$System.FN_Contrail_Setup.Dive Value)
       Front End=&Lerp_Float(A=-90, B=-100, Alpha=$System.FN_Contrail_Setup.Dive Value)
      )
     )
     Comparison Type=::NewEnumerator7
    )
    Trigger Condition:NiagaraBool=&SetBoolByFloatComparison(
     A=&MakeFloatFromVector(Vector=Linked() {linkedVariable {name=Particles.Position, type=NiagaraPosition}}, Channel=::NewEnumerator2)
     B=&Add_Float(
      A=&MakeFloatFromVector(Vector=Linked() {linkedVariable {name=Engine.Emitter.SimulationPosition, type=NiagaraPosition}}, Channel=::NewEnumerator2)
      B=&FN_FrontEndFloatModifier(
       Gameplay=&Lerp_Float(A=50, B=50, Alpha=$System.FN_Contrail_Setup.Dive Value)
       Front End=&Lerp_Float(A=60, B=60, Alpha=$System.FN_Contrail_Setup.Dive Value)
      )
     )
     Comparison Type=::NewEnumerator2
    )
   )
   &SolveForcesAndVelocity(Clamp Velocity=false, Limit Acceleration=false)
   /Niagara/Modules/Update/Utility/IncrementOverTime(Enabled:NiagaraBool=$Particles.DoOnce.TriggeredOnce)
   &AlignSpriteToMeshOrientation(SetAlignment=true, SetFacing=true)
   &GenerateLocationEvent(
    Event Generation Enabled:NiagaraBool=$Particles.DoOnce.Execute
    Event Probability=1
    Event Type:ENiagara_LocEventType=::NewEnumerator2
    Use Event Probability=true
    UseEventDelay=false
   )
   /SetVariables(Particles.PatternHighlightIntensity=5, Particles.PatternOpacity=1) {isSetParameters=true}
   &DynamicMaterialParameters(
    Index 0 Param 1=&FloatFromCurve(
     FloatCurve=DataInterface() {propertyValues {Curve {
        defaultValue=0
        keys [
         {arriveTangent=-0.0017, interpMode=RCIM_Cubic, leaveTangent=-0.0017, tangentMode=RCTM_User, time=0, value=1}
         {interpMode=RCIM_Cubic, tangentMode=RCTM_User, time=0.25, value=0.25}
         <at 0.8 -1>
        ]
       }}}
     CurveIndex=$StackContext.IncrementOverTime.Counter
    )
    Index 0 Param 2=&Lerp_Float(A=-1, B=-3, Alpha=1)
    Index 0 Param 3=&FloatFromCurve(FloatCurve=DataInterface() {propertyValues {Curve {keys [{interpMode=RCIM_Cubic, tangentMode=RCTM_User, time=0, value=1}, <at 0.3 0>]}}})
    Index 0 Param 4=&FloatFromCurve(FloatCurve=DataInterface() {propertyValues {Curve {keys [<at 0 1 RCIM_Cubic>, {interpMode=RCIM_Cubic, tangentMode=RCTM_User, time=0.15, value=1}, <at 1 1>]}}})
    Index 1 Param 1=&MakeCustomFloatFromBool(
     Bool=$Particles.Collided
     True Float=&FloatFromCurve(FloatCurve=DataInterface() {propertyValues {Curve {keys [{interpMode=RCIM_Cubic, tangentMode=RCTM_User, time=0.2, value=0}, <at 0.3 1.1>]}}})
     False Float=1
    )
    Module.Index0Float0Write=true
    Module.Index0Float1Write=true
    Module.Index0Float2Write=true
    Module.Index0Float3Write=true
    Module.Index1Float0Write=true
    Module.Index1Float1Write=false
    Module.Index1Float2Write=false
    Module.Index1Float3Write=false
    Write Parameter Index 1=::NewEnumerator1
   ) {module=DynamicMaterialParameters001}
   &ScaleColor(
    Scale Alpha=&Lerp_Float(
     A=&FloatFromCurve(
      FloatCurve=DataInterface() {propertyValues {Curve {keys [
          <at 0 0 RCIM_Cubic>
          {arriveTangent=4.3478, arriveTangentWeight=0.0333, interpMode=RCIM_Constant, leaveTangentWeight=0.0433, time=0.1, value=0.95}
          {interpMode=RCIM_Cubic, tangentMode=RCTM_User, time=0.25, value=0.95}
         ]}}}
      Scale Curve=/Game/Effects/Niagara/Modules/Math/FN_CameraDistance(
       0Dist:NiagaraFloat=64
       1Dist:NiagaraFloat=128
       Camera Query:NiagaraDataInterfaceCamera=DataInterface()
       Position:NiagaraPosition=$Particles.Position
      )
     )
     B=&FloatFromCurve(
      FloatCurve=DataInterface() {propertyValues {Curve {keys [
          {arriveTangent=-0.1583, interpMode=RCIM_Cubic, leaveTangent=-0.1583, tangentMode=RCTM_User, time=0, value=1}
          {arriveTangent=0.0616, interpMode=RCIM_Cubic, leaveTangent=0.0616, tangentMode=RCTM_User, time=0.5, value=0}
         ]}}}
      CurveIndex=$Particles.IncrementOverTime.Counter
     )
     Alpha=&MakeCustomFloatFromBool(Bool=$Particles.DoOnce.TriggeredOnce, True Float=1, False Float=0)
    )
    Scale RGB=&Lerp_Vector(
     A=&VectorFromFloat(Value=&FloatFromCurve(FloatCurve=DataInterface() {propertyValues {Curve {
          defaultValue=1
          keys [
           {arriveTangent=0.5854, interpMode=RCIM_Cubic, leaveTangent=0.5854, tangentMode=RCTM_User, time=0, value=0.2}
           {tangentMode=RCTM_User, time=1, value=0.5}
          ]
         }}}))
     B=&VectorFromCurve(
      VectorCurve=DataInterface() {propertyValues {
        XCurve {keys [
          {
           arriveTangent=-15.5376
           arriveTangentWeight=0.0667
           interpMode=RCIM_Cubic
           leaveTangent=-15.5376
           leaveTangentWeight=0.01
           tangentMode=RCTM_User
           time=0
           value=1.5
          }
          {interpMode=RCIM_Cubic, tangentMode=RCTM_User, time=0.2, value=0.5}
          {arriveTangent=-15.0637, interpMode=RCIM_Cubic, leaveTangent=-15.0637, tangentMode=RCTM_User, time=0.201, value=1.5}
          {arriveTangent=-1.0004, interpMode=RCIM_Cubic, leaveTangent=-1.0004, tangentMode=RCTM_User, time=0.4, value=0.5}
          <at 1 0.1 RCIM_Cubic>
         ]}
        YCurve {keys [
          {
           arriveTangent=-15.5376
           arriveTangentWeight=0.0667
           interpMode=RCIM_Cubic
           leaveTangent=-15.5376
           leaveTangentWeight=0.01
           tangentMode=RCTM_User
           time=0
           value=1.5
          }
          {interpMode=RCIM_Cubic, tangentMode=RCTM_User, time=0.2, value=0.5}
          {arriveTangent=-15.0637, interpMode=RCIM_Cubic, leaveTangent=-15.0637, tangentMode=RCTM_User, time=0.201, value=1.5}
          {arriveTangent=-1.0004, interpMode=RCIM_Cubic, leaveTangent=-1.0004, tangentMode=RCTM_User, time=0.4, value=0.5}
          <at 1 0.1 RCIM_Cubic>
         ]}
        ZCurve {keys [
          {
           arriveTangent=-15.5376
           arriveTangentWeight=0.0667
           interpMode=RCIM_Cubic
           leaveTangent=-15.5376
           leaveTangentWeight=0.01
           tangentMode=RCTM_User
           time=0
           value=1.5
          }
          {interpMode=RCIM_Cubic, tangentMode=RCTM_User, time=0.2, value=0.5}
          {arriveTangent=-15.0637, interpMode=RCIM_Cubic, leaveTangent=-15.0637, tangentMode=RCTM_User, time=0.201, value=1.5}
          {arriveTangent=-1.0004, interpMode=RCIM_Cubic, leaveTangent=-1.0004, tangentMode=RCTM_User, time=0.4, value=0.5}
          <at 1 0.1 RCIM_Cubic>
         ]}
       }}
      CurveIndex=$Particles.IncrementOverTime.Counter
     )
     Alpha=&MakeCustomFloatFromBool(Bool=$Particles.DoOnce.TriggeredOnce, True Float=1, False Float=0)
    )
    ScaleA=true
    ScaleRGB=true
   )
  ]
  rendererClasses [&NiagaraSpriteRendererProperties]
  renderers [{rendererClass=&NiagaraSpriteRendererProperties}]
 }
 HandGlow_Decal002 {
  bEnabled=false
  emitterUpdateScript [
   &EmitterState(Life Cycle Mode=::NewEnumerator1, Loop Behavior=::NewEnumerator1, Loop Duration Mode=::NewEnumerator1, UseLoopDelay=false)
   &SpawnBurst_Instantaneous(Spawn Count=4, Use Spawn Probability=false, UseLoopCountLimit=false)
  ]
  particleSpawnScript [
   /SetVariables(Particles.VisibilityTag=$Emitter.SpawnGroup) {isSetParameters=true}
   &PartitionParticles(Number Of Partitions=4)
   &InitializeParticle(
    Color Mode=::NewEnumerator0
    Sprite Size Mode=::NewEnumerator3
    Uniform Sprite Size=&SelectFloatFromArray(Direct Array Index=$Particles.PartitionParticles.Partition, Float Selection Array=DataInterface() {propertyValues {FloatData [55,55,40,40]}})
    UsePositionOffset=false
    Write Material Random=true
   )
  ]
  particleUpdateScript [
   &ParticleState(
    Kill Particles When Lifetime Has Elapsed=false
    Let Infinitely Lived Particles Die When Emitter Deactivates=true
    Loop Particles Lifetime=true
   )
   &SkeletalMeshLocation(
    Bone / Socket Index=&SelectIntFromArray(Direct Array Index=$Particles.PartitionParticles.Partition, Int Selection Array=DataInterface() {propertyValues {IntData [0,1,2,3]}})
    MaskSpawnGroup=false
    Mesh Sampling Type=::NewEnumerator5
    Skeletal Mesh=$System.Skeletal Mesh
    Socket Sampling Mode=::NewEnumerator5
   )
   &ScaleColor(
    Scale Alpha=&FloatFromCurve(
     FloatCurve=DataInterface() {propertyValues {Curve {keys [<at 0.0001 0.3382>,<at 0.1393 0.6251>,<at 0.3202 0.5233>,<at 0.4613 0.8934>,<at 0.6241 0.3845>,<at 0.7724 0.6066>,<at 1.0006 0.3567>]}}}
     Scale Curve=&SelectFloatFromArray(
      Direct Array Index=$Particles.PartitionParticles.Partition
      Float Selection Array=DataInterface() {propertyValues {FloatData [0.5,0.5,0.1,0.1]}}
     )
    )
    Scale RGB=&MakeVectorFromLinearColorRGB(Color=$User.FXColorA)
    ScaleA=true
    ScaleRGB=true
   )
   /SetVariables(
    Particles.LightColor=&MultiplyLinearColor(Linear Color=$User.FXColorA, Scale Factor=<1 1 1 20>)
    Particles.LightPosition:NiagaraPosition=&AddVectorToPosition(Position=$StackContext.SkeletalMeshLocation.SampledPosition, Vector=<0 0 22>)
   ) {isSetParameters=true}
  ]
  rendererClasses [&NiagaraSpriteRendererProperties, &NiagaraLightRendererProperties]
  renderers [
   {rendererClass=&NiagaraSpriteRendererProperties}
   {rendererClass=&NiagaraLightRendererProperties, rendererIndex=1}
  ]
 }
 Impact_Sparks {
  bEnabled=false
  emitterUpdateScript [&EmitterState()]
  particleSpawnScript [
   /SetVariables(
    Particles.IsNotGlowParticle:NiagaraBool=&SetBoolByFloatComparison(A=&ReturnNormalizedExecIndex(), B=0)
    Particles.ParticleIntensity:NiagaraFloat=&RandomRangeFloat(Minimum=0.2, Maximum=1)
    Particles.SpriteSize2=&RandomRangeVector2D(Minimum=<5 5>, Maximum=<7 7>)
   ) {isSetParameters=true}
   &InitializeParticle(
    Color=&MultiplyLinearColor(
     Linear Color=&LerpLinearColors(StartColor=$User.FXColorB, EndColor=$User.FXColorA, LerpFactor=$Particles.ParticleIntensity)
     Scale Factor=<5 5 5 1>
    )
    Lifetime=&MakeCustomFloatFromBool(
     Bool=$Particles.IsNotGlowParticle
     True Float=&Multiply_Float(
      A=&Lerp_Float(A=0.1, B=0.7, Alpha=$StackContext.ParticleIntensity)
      B=&FN_FrontEndFloatModifier(Gameplay=2, Front End=1.1)
     )
     False Float=0.23
    )
    Sprite Size Min=<7 5>
    Sprite Size Mode=::NewEnumerator2
    Sprite UV Mode=::NewEnumerator3
    UsePositionOffset=false
    Write Material Random=true
   )
  ]
  particleUpdateScript [
   /SetVariables(Particles.VisibilityTag=&MakeCustomIntFromBool(
     Bool=$Particles.Initial.IsNotGlowParticle
     True INT=&MakeCustomIntFromBool(Bool=&SetBoolByFloatComparison(A=$Particles.Initial.ParticleIntensity, B=0.5), True INT=1, False INT=2)
     False INT=0
    )) {isSetParameters=true}
   &ParticleState()
   &ScaleSpriteSize(
    Non-Uniform Curve Scale=&RandomRangeVector2D(Minimum=<0.5 0.5>, Maximum=<3 3>)
    Non-Uniform Curve Sprite Scale=DataInterface() {propertyValues {
      XCurve {keys [{interpMode=RCIM_Cubic, tangentMode=RCTM_User, time=0, value=1}, {tangentMode=RCTM_User, time=1, value=0.5}]}
      YCurve {keys [
        {arriveTangent=4.4047, interpMode=RCIM_Cubic, leaveTangent=4.4047, tangentMode=RCTM_User, time=-0.005, value=0.1}
        {arriveTangent=0.1609, leaveTangent=1.7823, tangentMode=RCTM_User, time=0.3, value=1}
        {interpMode=RCIM_Cubic, tangentMode=RCTM_User, time=0.5, value=1}
        {tangentMode=RCTM_User, time=1, value=0}
       ]}
     }}
    Scale Sprite Size Mode=::NewEnumerator3
   )
   &UpdateMeshOrientation(
    Facing Direction=$Particles.Velocity
    Facing Interpolation=false
    Orientation Method=::NewEnumerator1
    Override Quat=false
    Stabilization Mode:ENiagara_OrientVectorCount=::NewEnumerator2
    TransformFacing=false
    TransformSide=false
   )
   &Drag(Drag=3, Use Linear Drag=true, Use Rotational Drag=true)
   &AccelerationForce(Acceleration=<0 -100 500>, Coordinate Space=::Local)
   &SolveForcesAndVelocity(Clamp Velocity=false, Limit Acceleration=false)
   &ScaleSpriteSizeBySpeed(
    Max Scale Factor=&Multiply_Vector2D(
     A=&MakeVector2D(X=0.5, Y=&RandomRangeFloat(Minimum=0.2, Maximum=0.5))
     B=&Lerp_Vector2(A=<0.2 1>, B=<0.2 5>, Alpha=$Particles.Initial.ParticleIntensity)
    )
    Min Scale Factor=<0.1 1>
   )
   &AlignSpriteToMeshOrientation(
    Mesh Orientation Relative Sprite Alignment Vector=<1 0 0>
    Mesh Orientation Relative Sprite Facing Vector:Vector3f=<0 1 0>
    SetAlignment=true
    SetFacing=false
   )
   &ScaleColor(
    Scale Alpha=10
    Scale RGB=&Multiply_VectorByFloat(
     Vector=<1 1 1>
     Float=&FloatFromCurve(
      FloatCurve=DataInterface() {propertyValues {Curve {keys [{arriveTangent=-3098.343, interpMode=RCIM_Cubic, leaveTangent=-3098.343, tangentMode=RCTM_User, time=0, value=300}, <at 0.2 0.25>]}}}
      Scale Curve=&Add_Float(
       A=&RandomRangeFloat(Minimum=-0.5, Maximum=3, Evaluation Type=::NewEnumerator1)
       B=&FloatFromCurve(FloatCurve=DataInterface() {propertyValues {Curve {
           defaultValue=1
           keys [{interpMode=RCIM_Cubic, tangentMode=RCTM_User, time=0.7, value=1}, {arriveTangent=22.7661, tangentMode=RCTM_User, time=1, value=50}]
          }}})
      )
     )
    )
    ScaleA=true
    ScaleRGB=true
   )
   &FN_Sprite_Attributes(
    Camera Offset=10
    Color=&MultiplyLinearColor(
     Linear Color=$Particles.Initial.Color
     Scale Factor=&Vector4FromCurve(Vector4Curve=DataInterface() {propertyValues {
        AlphaCurve {keys [{arriveTangent=-0.4891, interpMode=RCIM_Cubic, leaveTangent=-0.4891, tangentMode=RCTM_User, time=0, value=0.3}, <at 1 0>]}
        BlueCurve {keys [{interpMode=RCIM_Cubic, tangentMode=RCTM_User, time=0, value=56.5}, <at 0.5 1>]}
        GreenCurve {keys [{interpMode=RCIM_Cubic, tangentMode=RCTM_User, time=0, value=56.5}, <at 0.5 1>]}
        RedCurve {keys [{interpMode=RCIM_Cubic, tangentMode=RCTM_User, time=0, value=56.7}, <at 0.5 1>]}
       }})
    )
    Create Camera Offset Attribute=true
    Create Color Attribute=true
    Create Position Attribute=true
    Create SpriteSize Attribute=true
    Position:NiagaraPosition=&VectorFromCurve(
     VectorCurve=DataInterface() {propertyValues {
       XCurve {keys [<at 0 1>,<at 1 1>]}
       YCurve {keys [<at 0 1>,<at 1 1>]}
       ZCurve {keys [{arriveTangent=1.3558, interpMode=RCIM_Cubic, leaveTangent=1.3558, tangentMode=RCTM_User, time=0, value=1}, <at 1 2>]}
      }}
     Scale Curve=<1 30 50>
    )
    SpriteSize=&Multiply_Vector2D_ByFloat(
     Float=&FloatFromCurve(FloatCurve=DataInterface() {propertyValues {Curve {keys [<at 0 0.7 RCIM_Cubic>, {interpMode=RCIM_Cubic, tangentMode=RCTM_User, time=0.2, value=1}, <at 1 0.3>]}}})
     Vector2D=<150 150>
    )
    Write to Module Namespace=true
   )
   /SetVariables(
    Particles.GlowColor=&ColorFromCurve(ColorCurve=DataInterface() {propertyValues {
       AlphaCurve {keys [<at 0 1 RCIM_Cubic>,<at 1 0 RCIM_Cubic>]}
       BlueCurve {keys [<at 0 0.0087 RCIM_Cubic>,<at 1 0.0019 RCIM_Cubic>]}
       GreenCurve {keys [<at 0 0.0087 RCIM_Cubic>,<at 1 0 RCIM_Cubic>]}
       RedCurve {keys [<at 0 0.3323 RCIM_Cubic>,<at 1 0.0573 RCIM_Cubic>]}
      }})
    Particles.SpriteSize2=&Multiply_Vector2D_ByFloat(
     Float=&FloatFromCurve(FloatCurve=DataInterface() {propertyValues {Curve {keys [{interpMode=RCIM_Cubic, tangentMode=RCTM_User, time=0, value=1}, <at 1 0>]}}})
     Vector2D=$Particles.Initial.SpriteSize2
    )
    Particles.SpriteSizeGlow:Vector2f=&Multiply_Vector2D(A=<3 1.3>, B=$Particles.SpriteSize)
   ) {isSetParameters=true}
  ]
  rendererClasses [&NiagaraSpriteRendererProperties]
  renderers [
   {rendererClass=&NiagaraSpriteRendererProperties}
   {rendererClass=&NiagaraSpriteRendererProperties, rendererIndex=1}
   {rendererClass=&NiagaraSpriteRendererProperties, rendererIndex=2}
   {rendererClass=&NiagaraSpriteRendererProperties, rendererIndex=3}
   {rendererClass=&NiagaraSpriteRendererProperties, rendererIndex=4}
  ]
 }
 Impact_Sparks_Rebuild {
  emitterUpdateScript [&EmitterState()]
  particleSpawnScript [
   /SetVariables(
    Particles.IsNotGlowParticle:NiagaraBool=&SetBoolByFloatComparison(A=&ReturnNormalizedExecIndex(), B=0)
    Particles.ParticleIntensity:NiagaraFloat=&RandomRangeFloat(Minimum=0.2, Maximum=1)
    Particles.SpriteSize2=&RandomRangeVector2D(Minimum=<5 5>, Maximum=<7 7>)
   ) {isSetParameters=true}
   &InitializeParticle(
    Color=&MultiplyLinearColor(
     Linear Color=&LerpLinearColors(StartColor=$User.FXColorB, EndColor=$User.FXColorA, LerpFactor=$Particles.ParticleIntensity)
     Scale Factor=<5 5 5 1>
    )
    Lifetime=&MakeCustomFloatFromBool(
     Bool=$Particles.IsNotGlowParticle
     True Float=&Multiply_Float(
      A=&Lerp_Float(A=0.1, B=0.7, Alpha=$StackContext.ParticleIntensity)
      B=&FN_FrontEndFloatModifier(Gameplay=2, Front End=1.1)
     )
     False Float=0.23
    )
    Sprite Size Min=<7 5>
    Sprite Size Mode=::NewEnumerator2
    Sprite UV Mode=::NewEnumerator3
    UsePositionOffset=false
    Write Material Random=true
   )
  ]
  particleUpdateScript [
   /SetVariables(Particles.VisibilityTag=&MakeCustomIntFromBool(
     Bool=$Particles.Initial.IsNotGlowParticle
     True INT=&MakeCustomIntFromBool(Bool=&SetBoolByFloatComparison(A=$Particles.Initial.ParticleIntensity, B=0.5), True INT=1, False INT=2)
     False INT=0
    )) {isSetParameters=true}
   &ParticleState()
   &ScaleSpriteSize(
    Non-Uniform Curve Scale=&RandomRangeVector2D(Minimum=<0.5 0.5>, Maximum=<3 3>)
    Non-Uniform Curve Sprite Scale=DataInterface() {propertyValues {
      XCurve {keys [{interpMode=RCIM_Cubic, tangentMode=RCTM_User, time=0, value=1}, {tangentMode=RCTM_User, time=1, value=0.5}]}
      YCurve {keys [
        {arriveTangent=4.4047, interpMode=RCIM_Cubic, leaveTangent=4.4047, tangentMode=RCTM_User, time=-0.005, value=0.1}
        {arriveTangent=0.1609, leaveTangent=1.7823, tangentMode=RCTM_User, time=0.3, value=1}
        {interpMode=RCIM_Cubic, tangentMode=RCTM_User, time=0.5, value=1}
        {tangentMode=RCTM_User, time=1, value=0}
       ]}
     }}
    Scale Sprite Size Mode=::NewEnumerator3
   )
   &UpdateMeshOrientation(
    Facing Direction=$Particles.Velocity
    Facing Interpolation=false
    Orientation Method=::NewEnumerator1
    Override Quat=false
    Stabilization Mode:ENiagara_OrientVectorCount=::NewEnumerator2
    TransformFacing=false
    TransformSide=false
   )
   &Drag(Drag=3, Use Linear Drag=true, Use Rotational Drag=true)
   &AccelerationForce(Acceleration=<0 -100 500>, Coordinate Space=::Local)
   &SolveForcesAndVelocity(Clamp Velocity=false, Limit Acceleration=false)
   &ScaleSpriteSizeBySpeed(
    Max Scale Factor=&Multiply_Vector2D(
     A=&MakeVector2D(X=0.5, Y=&RandomRangeFloat(Minimum=0.2, Maximum=0.5))
     B=&Lerp_Vector2(A=<0.2 1>, B=<0.2 5>, Alpha=$Particles.Initial.ParticleIntensity)
    )
    Min Scale Factor=<0.1 1>
   )
   &AlignSpriteToMeshOrientation(
    Mesh Orientation Relative Sprite Alignment Vector=<1 0 0>
    Mesh Orientation Relative Sprite Facing Vector:Vector3f=<0 1 0>
    SetAlignment=true
    SetFacing=false
   )
   &ScaleColor(
    Scale Alpha=10
    Scale RGB=&Multiply_VectorByFloat(
     Vector=<1 1 1>
     Float=&FloatFromCurve(
      FloatCurve=DataInterface() {propertyValues {Curve {keys [{arriveTangent=-3098.343, interpMode=RCIM_Cubic, leaveTangent=-3098.343, tangentMode=RCTM_User, time=0, value=300}, <at 0.2 0.25>]}}}
      Scale Curve=&Add_Float(
       A=&RandomRangeFloat(Minimum=-0.5, Maximum=3, Evaluation Type=::NewEnumerator1)
       B=&FloatFromCurve(FloatCurve=DataInterface() {propertyValues {Curve {
           defaultValue=1
           keys [{interpMode=RCIM_Cubic, tangentMode=RCTM_User, time=0.7, value=1}, {arriveTangent=22.7661, tangentMode=RCTM_User, time=1, value=50}]
          }}})
      )
     )
    )
    ScaleA=true
    ScaleRGB=true
   )
   &FN_Sprite_Attributes(
    Camera Offset=10
    Color=&MultiplyLinearColor(
     Linear Color=$Particles.Initial.Color
     Scale Factor=&Vector4FromCurve(Vector4Curve=DataInterface() {propertyValues {
        AlphaCurve {keys [{arriveTangent=-0.4891, interpMode=RCIM_Cubic, leaveTangent=-0.4891, tangentMode=RCTM_User, time=0, value=0.3}, <at 1 0>]}
        BlueCurve {keys [{interpMode=RCIM_Cubic, tangentMode=RCTM_User, time=0, value=56.5}, <at 0.5 1>]}
        GreenCurve {keys [{interpMode=RCIM_Cubic, tangentMode=RCTM_User, time=0, value=56.5}, <at 0.5 1>]}
        RedCurve {keys [{interpMode=RCIM_Cubic, tangentMode=RCTM_User, time=0, value=56.7}, <at 0.5 1>]}
       }})
    )
    Create Camera Offset Attribute=true
    Create Color Attribute=true
    Create Position Attribute=true
    Create SpriteSize Attribute=true
    Position:NiagaraPosition=&VectorFromCurve(
     VectorCurve=DataInterface() {propertyValues {
       XCurve {keys [<at 0 1>,<at 1 1>]}
       YCurve {keys [<at 0 1>,<at 1 1>]}
       ZCurve {keys [{arriveTangent=1.3558, interpMode=RCIM_Cubic, leaveTangent=1.3558, tangentMode=RCTM_User, time=0, value=1}, <at 1 2>]}
      }}
     Scale Curve=<1 30 50>
    )
    SpriteSize=&Multiply_Vector2D_ByFloat(
     Float=&FloatFromCurve(FloatCurve=DataInterface() {propertyValues {Curve {keys [<at 0 0.7 RCIM_Cubic>, {interpMode=RCIM_Cubic, tangentMode=RCTM_User, time=0.2, value=1}, <at 1 0.3>]}}})
     Vector2D=<150 150>
    )
    Write to Module Namespace=true
   )
   /SetVariables(
    Particles.GlowColor=&ColorFromCurve(ColorCurve=DataInterface() {propertyValues {
       AlphaCurve {keys [<at 0 1 RCIM_Cubic>,<at 1 0 RCIM_Cubic>]}
       BlueCurve {keys [<at 0 0.0087 RCIM_Cubic>,<at 1 0.0019 RCIM_Cubic>]}
       GreenCurve {keys [<at 0 0.0087 RCIM_Cubic>,<at 1 0 RCIM_Cubic>]}
       RedCurve {keys [<at 0 0.3323 RCIM_Cubic>,<at 1 0.0573 RCIM_Cubic>]}
      }})
    Particles.SpriteSize2=&Multiply_Vector2D_ByFloat(
     Float=&FloatFromCurve(FloatCurve=DataInterface() {propertyValues {Curve {keys [{interpMode=RCIM_Cubic, tangentMode=RCTM_User, time=0, value=1}, <at 1 0>]}}})
     Vector2D=$Particles.Initial.SpriteSize2
    )
    Particles.SpriteSizeGlow:Vector2f=&Multiply_Vector2D(A=<3 1.3>, B=$Particles.SpriteSize)
   ) {isSetParameters=true}
  ]
  rendererClasses [&NiagaraSpriteRendererProperties]
  renderers [
   {rendererClass=&NiagaraSpriteRendererProperties}
   {rendererClass=&NiagaraSpriteRendererProperties, rendererIndex=1}
   {rendererClass=&NiagaraSpriteRendererProperties, rendererIndex=2}
   {rendererClass=&NiagaraSpriteRendererProperties, rendererIndex=3}
   {rendererClass=&NiagaraSpriteRendererProperties, rendererIndex=4}
  ]
 }
 Trail_Hands {
  emitterUpdateScript [
   &EmitterState(Life Cycle Mode=::NewEnumerator1, Loop Behavior=::NewEnumerator1, Loop Duration Mode=::NewEnumerator1, UseLoopDelay=false)
   &SpawnBurst_Instantaneous(Spawn Count=2, Use Spawn Probability=true, UseLoopCountLimit=false)
  ]
  particleSpawnScript [
   &PartitionParticles()
   &InitializeParticle(
    Color=<rgba 0.4115 0 0 1>
    Sprite Size=&MakeVector2D(
     X=&Multiply_Float(
      A=100
      B=&Lerp_Float(A=1, B=-1, Alpha=&MakeFloatFromInt(Int=$Particles.PartitionParticles.Partition))
     )
     Y=150
    )
    Sprite Size Mode=::NewEnumerator4
    UsePositionOffset=false
    Write Material Random=true
   )
  ]
  particleUpdateScript [
   &ParticleState(
    Kill Particles When Lifetime Has Elapsed=false
    Let Infinitely Lived Particles Die When Emitter Deactivates=true
    Loop Particles Lifetime=true
   )
   &ScaleSpriteSize(
    Scale Factor=&Multiply_Vector2D(
     A=&Lerp_Vector2(A=<0.5 0.8>, B=<0.6 1.5>, Alpha=$System.FN_Contrail_Setup.Dive Value)
     B=&RandomRangeVector2D(Minimum=<0.95 0.9>, Maximum=<1.05 1.1>, Evaluation Type=::NewEnumerator1)
    )
    Scale Sprite Size Mode=::NewEnumerator2
   )
   &ScaleColor(
    Scale Alpha=&FloatFromCurve(
     FloatCurve=DataInterface() {propertyValues {Curve {
        defaultValue=1
        keys [
         {interpMode=RCIM_Cubic, tangentMode=RCTM_User, time=0, value=0.8}
         {interpMode=RCIM_Cubic, tangentMode=RCTM_User, time=0.5, value=1}
         {tangentMode=RCTM_User, time=1, value=0.8}
        ]
       }}}
     CurveIndex=Linked() {linkedVariable {name=NPC.FortniteNPC.TimeOfDay_0-1, type=NiagaraFloat}}
    )
    Scale RGB=&Multiply_VectorByFloat(
     Vector=<1 1 1>
     Float=&FloatFromCurve(
      FloatCurve=DataInterface() {propertyValues {Curve {keys [
          {interpMode=RCIM_Cubic, tangentMode=RCTM_User, time=0, value=1}
          {arriveTangent=1.631, interpMode=RCIM_Cubic, leaveTangent=1.631, tangentMode=RCTM_User, time=1, value=2}
         ]}}}
      CurveIndex=$System.FN_Contrail_Setup.Dive Value
     )
    )
    ScaleA=true
    ScaleRGB=true
   )
   &CameraOffset(Camera Offset Amount=-5)
   &SkeletalMeshLocation(
    Bone / Socket Index=$Particles.PartitionParticles.Partition
    MaskSpawnGroup=false
    Mesh Sampling Type=::NewEnumerator5
    Orientation Sampling=::NewEnumerator1
    Sampled Position Offset=&FN_FrontEndVector3Modifier(
     Gameplay=&MakeVector(
      X=&Lerp_Float(A=2, B=10, Alpha=$User.ParaGlideLeanAlpha)
      Y=&Lerp_Float(A=-0.1, B=0, Alpha=$User.ParaGlideLeanAlpha)
      Z=&Multiply_Float(
       A=&Lerp_Float(A=0, B=5, Alpha=$User.ParaGlideLeanAlpha)
       B=&Lerp_Float(A=1, B=-1, Alpha=&MakeFloatFromInt(Int=$Particles.PartitionParticles.Partition))
      )
     )
     Front End=&MakeVector(
      X=&Lerp_Float(A=2, B=5, Alpha=$User.ParaGlideLeanAlpha)
      Y=&Lerp_Float(A=0, B=0, Alpha=$User.ParaGlideLeanAlpha)
      Z=&Multiply_Float(
       A=&Lerp_Float(A=0, B=3, Alpha=$User.ParaGlideLeanAlpha)
       B=&Lerp_Float(A=1, B=-1, Alpha=&MakeFloatFromInt(Int=$Particles.PartitionParticles.Partition))
      )
     )
    )
    Skeletal Mesh=$System.Skeletal Mesh
    Socket Sampling Mode=::NewEnumerator5
   )
   /SetVariables(
    Particles.CameraOffset2:NiagaraFloat=&Add_Float(A=$Particles.CameraOffset, B=20)
    Particles.Color2=&LerpLinearColors(
     StartColor=<rgba 500 500 500 0.5>
     EndColor=&MultiplyLinearColor(Linear Color=$Particles.Color, Scale Factor=<1 1 1 2>)
     LerpFactor=$System.FN_Contrail_Setup.Dive Value
    )
    Particles.Color3:LinearColor=&Multiply_LinearColorByFloat(
     Linear Color=&LerpLinearColors(StartColor=<rgba 0.5 0 0 0.3>, EndColor=<rgba 0.7 0 0 0.5>, LerpFactor=$User.ParaGlideLeanAlpha)
     Float=&FloatFromCurve(
      FloatCurve=DataInterface() {propertyValues {Curve {
         defaultValue=1
         keys [
          {interpMode=RCIM_Cubic, tangentMode=RCTM_User, time=0.5, value=1}
          {arriveTangent=-0.1507, interpMode=RCIM_Cubic, leaveTangent=-0.1507, tangentMode=RCTM_User, time=1, value=1}
         ]
        }}}
      CurveIndex=$Emitter.Age
     )
    )
    Particles.GlowPosition:NiagaraPosition=&AddVectorToPosition(
     Position=$Particles.Position
     Vector=&Multiply_VectorByFloat(
      Vector=&RandomRangeVector(Minimum=<-1 -1 -1>, Maximum=<1 1 1>, Evaluation Type=::NewEnumerator1)
      Float=1
     )
    )
    Particles.SpriteSize2=&Multiply_Vector2D(
     A=$Particles.SpriteSize
     B=&MakeVector2D(X=&RandomRangeFloat(Minimum=0.95, Maximum=1.2, Evaluation Type=::NewEnumerator1), Y=0.6)
    )
    Particles.SpriteSize3:Vector2f=<15 15>
   ) {isSetParameters=true}
   /SetVariables(
    Particles.Color4:LinearColor=&Multiply_LinearColorByFloat(Linear Color=$Particles.Color3, Float=2)
    Particles.DecalSize:Vector2f=<20 20>
   ) {isSetParameters=true}
   &UpdateMeshOrientation(
    Facing Direction=$System.FN_Contrail_Setup.Direction Vector
    Facing Interpolation=false
    Orientation Method=::NewEnumerator1
    Override Quat=false
    Stabilization Mode:ENiagara_OrientVectorCount=::NewEnumerator2
    TransformFacing=false
    TransformSide=false
   )
   &AlignSpriteToMeshOrientation(
    Mesh Orientation Relative Sprite Alignment Vector=<-1 0 0>
    Mesh Orientation Relative Sprite Facing Vector:Vector3f=<0 1 0>
    SetAlignment=true
    SetFacing=true
   )
   &DynamicMaterialParameters(
    Index 0 Param 1=&FloatFromCurve(
     FloatCurve=DataInterface() {propertyValues {Curve {
        defaultValue=1
        keys [
         {interpMode=RCIM_Cubic, tangentMode=RCTM_User, time=0.5, value=1}
         {arriveTangent=-6.112, interpMode=RCIM_Cubic, leaveTangent=-6.112, tangentMode=RCTM_User, time=1, value=0}
        ]
       }}}
     CurveIndex=$Emitter.Age
    )
    Index 0 Param 2=&FloatFromCurve(
     FloatCurve=DataInterface() {propertyValues {Curve {
        defaultValue=0
        keys [
         {interpMode=RCIM_Cubic, tangentMode=RCTM_User, time=0, value=0}
         {arriveTangent=1, interpMode=RCIM_Cubic, leaveTangent=1, time=0.5, value=0.5}
         {arriveTangent=0.0846, interpMode=RCIM_Cubic, leaveTangent=0.0846, tangentMode=RCTM_User, time=1, value=1}
        ]
       }}}
     CurveIndex=$Emitter.Age
    )
    Index 0 Param 3=&Lerp_Float(A=10, B=20, Alpha=$System.FN_Contrail_Setup.Dive Value)
    Module.Index0Float0Write=true
    Module.Index0Float1Write=true
    Module.Index0Float2Write=true
    Module.Index0Float3Write=false
   ) {module=DynamicMaterialParameters001}
   &Color(
    Color=&Multiply_LinearColorByFloat(
     Linear Color=&LerpLinearColors(StartColor=$User.FXColorB, EndColor=$User.FXColorA, LerpFactor=$System.FN_Contrail_Setup.Dive Value)
     Float=&FloatFromCurve(
      FloatCurve=DataInterface() {propertyValues {Curve {keys [{interpMode=RCIM_Cubic, tangentMode=RCTM_User, time=0, value=2}, <at 1 10>]}}}
      CurveIndex=$User.ParaGlideLeanAlpha
     )
    )
    Scale Color=&Multiply_VectorByFloat(
     Vector=<1 1 1>
     Float=&FloatFromCurve(FloatCurve=DataInterface() {propertyValues {Curve {keys [{arriveTangent=-17.669, interpMode=RCIM_Cubic, leaveTangent=-17.669, tangentMode=RCTM_User, time=0, value=100}, <at 0.2 1>, <at 1 1>]}}})
    )
   )
  ]
  rendererClasses [&NiagaraSpriteRendererProperties]
  renderers [
   {rendererClass=&NiagaraSpriteRendererProperties}
   {rendererClass=&NiagaraSpriteRendererProperties, rendererIndex=1}
   {rendererClass=&NiagaraSpriteRendererProperties, rendererIndex=2}
   {rendererClass=&NiagaraSpriteRendererProperties, rendererIndex=3}
  ]
 }
}
summary {
 systemName=NS_Contrail_FoundKelp
 emitters [
  {emitterName=Trail_Hands, rendererClasses [&NiagaraSpriteRendererProperties]}
  {
   bEnabled=false
   emitterName=HandGlow_Decal002
   rendererClasses [&NiagaraSpriteRendererProperties, &NiagaraLightRendererProperties]
  }
  {emitterName=DigitalSquares001, rendererClasses [&NiagaraSpriteRendererProperties]}
  {bEnabled=false, emitterName=Impact_Sparks, rendererClasses [&NiagaraSpriteRendererProperties]}
  {emitterName=FoundKelp_Logo_Mesh, rendererClasses [&NiagaraSpriteRendererProperties]}
  {emitterName=Impact_Sparks_Rebuild, rendererClasses [&NiagaraSpriteRendererProperties]}
 ]
 userVariables [
  User.bIsFrontEnd {type=NiagaraBool}
  User.bIsFrontEndPreview {type=NiagaraBool}
  User.FXColorA {type=LinearColor, description="", defaultValue {struct=LinearColor, value=<rgba 0.1573 0.0019 0 1>}}
  User.FXColorB {type=LinearColor, description="", defaultValue {struct=LinearColor, value=<rgba 0.3003 0 0.0045 1>}}
  User.ParaGlideLeanAlpha {type=NiagaraFloat, defaultValue=1, description=""}
 ]
}
